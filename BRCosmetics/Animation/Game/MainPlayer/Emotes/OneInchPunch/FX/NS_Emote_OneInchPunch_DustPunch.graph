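# Niagara system for the One Inch Punch emote: a burst of dust, flash, impact spikes and an airflow mesh at the punch point.
# Each emitter samples the skeletal mesh and spawns once when the sampled X crosses a threshold, edge-detected against last frame's boolean.
# DustPunch sends two spawn groups along ring tangents, animated by dynamic material parameters.
# Time-of-day scalars adjust brightness and opacity; front-end modifiers adjust dust offset and drift.
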NiagaraSystem /BRCosmetics/Animation/Game/MainPlayer/Emotes/OneInchPunch/FX/NS_Emote_OneInchPunch_DustPunch
dependencies {
 usedDataInterfaces [
  &NiagaraDataInterfaceCurve
  &NiagaraDataInterfaceCamera
  &NiagaraDataInterfaceVector2DCurve
  &NiagaraDataInterfaceColorCurve
  &NiagaraDataInterfaceSkeletalMesh
 ]
 usedDynamicInputs [&SimulationPosition]
 usedModules [
  &self:SystemSpawnScript.SystemScriptSource.SystemScriptGraph.NiagaraNodeAssignment_2.SetVariables_#1
  &SystemState
  &self:SystemSpawnScript.SystemScriptSource.SystemScriptGraph.NiagaraNodeAssignment_1.SetVariables_#2
 ]
}
emitters {
 Airflow {
  emitterUpdateScript [
   &EmitterState(Life Cycle Mode=::NewEnumerator1, Loop Behavior=::NewEnumerator1, Loop Delay=0.08, Loop Duration Mode=::NewEnumerator1, UseLoopDelay=true)
   &SampleSkeletalMesh(
    AdditionalYawPitchRoll=false
    Bone Sampling Mode:ENiagaraBoneSamplingMode=::NewEnumerator4
    Position Offset Coordinate Space:ENiagaraCoordinateSpace=::Local
    Sampled Position Offset:Vector3f=<-10 5 0>
    Skeletal Mesh:NiagaraDataInterfaceSkeletalMesh=$System.Skeletal Mesh
   )
   /SetVariables(Emitter.bool:NiagaraBool=&SetBoolByFloatComparison(
     A=&MakeFloatFromVector(Vector=$Emitter.SampleSkeletalMesh.SampledPositionWithoutOffset, Channel=::NewEnumerator0)
     B=-47
     Comparison Type=::NewEnumerator6
    )) {isSetParameters=true, module=SetVariables_#3}
   /SetVariables(Emitter.Spawn:NiagaraBool=&Boolean_AndOperation(Bool1=&Boolean_NotOperation(Bool=$Emitter.boolLastFrame), Bool2=$Emitter.bool)) {isSetParameters=true, module=SetVariables_#4}
   /SetVariables(Emitter.boolLastFrame:NiagaraBool=$Emitter.bool) {isSetParameters=true, module=SetVariables_#5}
   &SpawnPerFrame(Spawn:NiagaraBool=$Emitter.Spawn, Use Spawn Probability=false)
  ]
  particleSpawnScript [
   &InitializeParticle(
    Lifetime=0.2
    Mesh Scale=&Multiply_VectorByFloat(Vector=&RandomRangeVector(Minimum=<0.12 0.14 0.14>, Maximum=<0.14 0.14 0.14>), Float=0.6)
    Mesh Scale Mode=::NewEnumerator4
    Position=$Emitter.SampleSkeletalMesh.SampledPosition
    Position Mode=::NewEnumerator1
    UsePositionOffset=false
    Write Material Random=true
   )
   &InitialMeshOrientation(Mesh Orientation Mode=::NewEnumerator2, Rotation=<0.1 0.5 0>, Use Rotation Vector=true)
   &CameraOffset(Camera Offset Amount=30) {enabled=false}
  ]
  particleUpdateScript [
   &ParticleState()
   &ScaleMeshSize(Scale Factor=&MakeVector(
     X=&FloatFromCurve(FloatCurve=DataInterface() {propertyValues {Curve {keys [{arriveTangent=0.3763, interpMode=RCIM_Cubic, leaveTangent=0.3763, tangentMode=RCTM_User, time=0, value=1}, <at 1 1.2>]}}})
     Y=&FloatFromCurve(FloatCurve=DataInterface() {propertyValues {Curve {keys [{arriveTangent=3.1084, interpMode=RCIM_Cubic, leaveTangent=3.1084, tangentMode=RCTM_User, time=0, value=0.5}, <at 0.3 1>]}}})
     Z=&FloatFromCurve(FloatCurve=DataInterface() {propertyValues {Curve {keys [{arriveTangent=3.1084, interpMode=RCIM_Cubic, leaveTangent=3.1084, tangentMode=RCTM_User, time=0, value=0.5}, <at 0.3 1>]}}})
    ))
   &ScaleColor(
    Scale Alpha=&FloatFromCurve(
     FloatCurve=DataInterface() {propertyValues {Curve {keys [<at 0 1>,<at 1 0>]}}}
     Scale Curve=&MakeCustomFloatFromBool(
      Bool=$User.bIsFrontEndPreview
      True Float=0.5
      False Float=&FNTimeOfDayScalar(Night Scale=0.5, Morning Scale=1, Evening Scale=1)
     )
    )
    Scale RGB=&Multiply_Vector(
     A=&MakeVectorFromLinearColorRGB(Color=&ColorFromCurve(ColorCurve=DataInterface() {propertyValues {
         AlphaCurve {keys []}
         BlueCurve {keys [<at 0 0.6 RCIM_Cubic>,<at 0.5 0.444 RCIM_Cubic>]}
         GreenCurve {keys [<at 0 0.7133 RCIM_Cubic>,<at 0.5 0.4599 RCIM_Cubic>]}
         RedCurve {keys [<at 0 1 RCIM_Cubic>,<at 0.5 0.5 RCIM_Cubic>]}
        }}))
     B=&VectorFromFloat(Value=&FNTimeOfDayScalar(Night Scale=1, Morning Scale=1, Evening Scale=1))
    )
    ScaleA=true
    ScaleRGB=true
   )
   /SetVariables(Particles.Dy3:Vector4f=&MakeVector4(X=0, Y=1, Z=0, W=1)) {enabled=false, isSetParameters=true}
   &ScaleColor(
    Scale RGB=&MakeVectorFromLinearColorRGB(Color=&FN_Select_Color_from_Time_Of_Day(Morning=<rgba 0.8 0.7415 0.6111 1>, Day=<rgba 1 1 1 1>, Evening=<rgba 0.25 0.2307 0.225 1>, Night=<rgba 0.2451 0.5043 1 1>))
    ScaleA=false
    ScaleRGB=false
   ) {module=ScaleColor001}
  ]
  rendererClasses [&NiagaraMeshRendererProperties]
  renderers [{rendererClass=&NiagaraMeshRendererProperties, rendererIndex=0}]
 }
 DustPunch {
  emitterUpdateScript [
   &EmitterState(Life Cycle Mode=::NewEnumerator1, Loop Behavior=::NewEnumerator1, Loop Duration Mode=::NewEnumerator1, UseLoopDelay=false)
   &SampleSkeletalMesh(
    AdditionalYawPitchRoll=false
    Bone Sampling Mode:ENiagaraBoneSamplingMode=::NewEnumerator4
    Position Offset Coordinate Space:ENiagaraCoordinateSpace=::Local
    Sampled Position Offset:Vector3f=<-10 7 0>
    Skeletal Mesh:NiagaraDataInterfaceSkeletalMesh=$System.Skeletal Mesh
   )
   /SetVariables(Emitter.bool:NiagaraBool=&SetBoolByFloatComparison(
     A=&MakeFloatFromVector(Vector=$Emitter.SampleSkeletalMesh.SampledPositionWithoutOffset, Channel=::NewEnumerator0)
     B=-47
     Comparison Type=::NewEnumerator6
    )) {isSetParameters=true, module=SetVariables_#3}
   /SetVariables(Emitter.Spawn:NiagaraBool=&Boolean_AndOperation(Bool1=&Boolean_NotOperation(Bool=$Emitter.boolLastFrame), Bool2=$Emitter.bool)) {isSetParameters=true, module=SetVariables_#4}
   /SetVariables(Emitter.boolLastFrame:NiagaraBool=$Emitter.bool) {isSetParameters=true, module=SetVariables_#5}
   &SpawnPerFrame(Spawn:NiagaraBool=$Emitter.Spawn, Spawn Count=3, Use Spawn Probability=false)
   &SpawnPerFrame(Spawn:NiagaraBool=$Emitter.Spawn, Spawn Count=3, Spawn Group=1, Use Spawn Probability=false) {module=SpawnPerFrame001}
  ]
  particleSpawnScript [
   &InitializeParticle(
    Color=<rgba 0.15 0.132 0.126 0.25>
    Lifetime Max=1
    Lifetime Min=0.6
    Lifetime Mode=::NewEnumerator1
    Position:Vector3f=$Emitter.SampleSkeletalMesh.SampledPosition
    Position Mode=::NewEnumerator1
    Position Offset=&FN_FrontEndVector3Modifier(Gameplay=<0 0 0>, Front End=<0 -16 0>)
    Sprite Rotation Mode=::NewEnumerator2
    Sprite Size Mode=::NewEnumerator1
    Uniform Sprite Size Max=33
    Uniform Sprite Size Min=22
    UsePositionOffset=false
    Write Material Random=true
   )
   &ShapeLocation(
    MaskSpawnGroup=true
    Override Seed=false
    Ring Radius=1
    Rotation Angle=0.5
    Rotation Mode=::NewEnumerator1
    Shape Primitive=::NewEnumerator4
    U Distribution=0.5
    UseOwnerScale=false
   )
   &ShapeLocation(
    MaskSpawnGroup=true
    Override Seed=false
    Ring Radius=1
    Rotation Angle=1
    Rotation Mode=::NewEnumerator1
    Shape Primitive=::NewEnumerator4
    Spawn Group Mask=1
    U Distribution=0.5
    UseOwnerScale=false
   ) {module=ShapeLocation001}
   &AddVelocity(
    Velocity=&MaskVectorBySpawnGroup(Masked Vector=$Particles.ShapeLocation.ShapeTangent, Passthrough Vector=$Particles.ShapeLocation001.ShapeTangent)
    Velocity Speed Scale=&RandomRangeFloat(Minimum=111, Maximum=222)
   )
   &AddVelocity(Velocity=<0 0 0>, Velocity Speed Scale=&RandomRangeFloat(Minimum=25, Maximum=50)) {module=AddVelocity001}
  ]
  particleUpdateScript [
   &CameraOffset(Camera Offset Amount=&FloatFromCurve(
     FloatCurve=DataInterface() {propertyValues {Curve {keys [{tangentMode=RCTM_User, time=0, value=0}, {tangentMode=RCTM_User, time=1, value=1}]}}}
     Scale Curve=44
    ))
   &Drag(Drag=15, Rotational Drag=0, Use Linear Drag=true, Use Rotational Drag=false)
   &GravityForce(Gravity=<0 0 33>)
   &ParticleState()
   &AccelerationForce(
    Acceleration=&FN_FrontEndVector3Modifier(Gameplay=<0 0 0>, Front End=<0 -222 0>)
    Coordinate Space=::Local
   )
   &SolveForcesAndVelocity(Clamp Velocity=false, Limit Acceleration=false)
   &DynamicMaterialParameters(
    Index 0 Param 1=&Add_Float(
     A=$Particles.RandomFloat
     B=&FloatFromCurve(
      FloatCurve=DataInterface() {propertyValues {Curve {keys [
          {arriveTangent=0.9088, interpMode=RCIM_Cubic, leaveTangent=0.9088, tangentMode=RCTM_User, time=0, value=1}
          {arriveTangent=0.5473, interpMode=RCIM_Cubic, leaveTangent=0.5473, tangentMode=RCTM_User, time=1, value=2}
         ]}}}
      CurveIndex=$Particles.Age
      Scale Curve=0.222
     )
    )
    Index 0 Param 2=&FloatFromCurve(FloatCurve=DataInterface() {propertyValues {Curve {keys [
         {arriveTangent=0.0453, interpMode=RCIM_Cubic, tangentMode=RCTM_Break, time=0, value=-1}
         {arriveTangent=0.0604, interpMode=RCIM_Cubic, leaveTangent=0.0604, tangentMode=RCTM_User, time=1, value=-1}
        ]}}})
    Index 0 Param 3=&FloatFromCurve(FloatCurve=DataInterface() {propertyValues {Curve {keys [
         {arriveTangent=-9.0447, interpMode=RCIM_Cubic, leaveTangent=-2.0044, tangentMode=RCTM_User, time=0, value=0.5}
         {interpMode=RCIM_Cubic, leaveTangent=4.6779, time=0.2851, value=0}
        ]}}})
    Index 1 Param 1=&FloatFromCurve(
     FloatCurve=DataInterface() {propertyValues {Curve {keys [<at 0 0>,<at 0.25 1>]}}}
     Scale Curve=&RandomRangeFloat(Minimum=2, Maximum=1)
    )
    Index 1 Param 2=&FloatFromCurve(
     FloatCurve=DataInterface() {propertyValues {Curve {keys [
         {interpMode=RCIM_Cubic, leaveTangent=3.7422, tangentMode=RCTM_Break, time=0, value=0.2}
         {interpMode=RCIM_Cubic, leaveTangent=0.9066, time=1, value=2}
        ]}}}
     Scale Curve=&RandomRangeFloat(Minimum=0.5, Maximum=1)
    )
    Index 1 Param 3=&FloatFromCurve(FloatCurve=DataInterface() {propertyValues {Curve {keys [
         {interpMode=RCIM_Cubic, leaveTangent=2.8445, tangentMode=RCTM_Break, time=0, value=0.2}
         {interpMode=RCIM_Cubic, leaveTangent=63.2978, time=0.25, value=1}
        ]}}})
    Index 1 Param 4=&FloatFromCurve(
     FloatCurve=DataInterface() {propertyValues {Curve {keys [<at 0 1>,<at 1 0.8>]}}}
     Scale Curve=&RandomRangeFloat(Minimum=0.5, Maximum=1)
    )
    Index 2 Param 1=&FloatFromCurve(
     FloatCurve=DataInterface() {propertyValues {Curve {keys [{interpMode=RCIM_Cubic, leaveTangent=3.2309, tangentMode=RCTM_Break, time=0.1, value=0}, <at 1 1>]}}}
     Scale Curve=0
    )
    Index 2 Param 2=&FloatFromCurve(
     FloatCurve=DataInterface() {propertyValues {Curve {keys [<at 0 0.2 RCIM_Cubic>,<at 1 0.8 RCIM_Cubic>]}}}
     Scale Curve=&RandomRangeFloat(Minimum=0.2, Maximum=1.2)
    )
    Index 2 Param 3=&FloatFromCurve(FloatCurve=DataInterface() {propertyValues {Curve {keys [
         {interpMode=RCIM_Cubic, tangentMode=RCTM_Break, time=0, value=1}
         {interpMode=RCIM_Cubic, leaveTangent=11.1233, tangentMode=RCTM_Break, time=1, value=2}
        ]}}})
    Index 2 Param 4=&FloatFromCurve(
     FloatCurve=DataInterface() {propertyValues {Curve {keys [<at 0 1 RCIM_Cubic>, {interpMode=RCIM_Cubic, leaveTangent=-1.2956, tangentMode=RCTM_Break, time=1, value=0.75}]}}}
     Scale Curve=0.8
    )
    Module.Index0Param0Write=true
    Module.Index0Param1Write=true
    Module.Index0Param2Write=true
    Module.Index0Param3Write=true
    Module.Index1Param0Write=true
    Module.Index1Param1Write=true
    Module.Index1Param2Write=true
    Module.Index1Param3Write=true
    Module.Index2Param0Write=true
    Module.Index2Param1Write=true
    Module.Index2Param2Write=true
    Module.Index2Param3Write=true
    Write Parameter Index 1=true
    Write Parameter Index 2=true
   )
   &ScaleSpriteSize(Uniform Curve Sprite Scale=DataInterface() {propertyValues {Curve {keys [
        {arriveTangent=4.1785, interpMode=RCIM_Cubic, leaveTangent=4.1785, tangentMode=RCTM_User, time=0, value=0}
        {arriveTangent=1.1179, interpMode=RCIM_Cubic, leaveTangent=1.1179, tangentMode=RCTM_User, time=0.2042, value=0.6228}
        {arriveTangent=0.0224, interpMode=RCIM_Cubic, leaveTangent=0.0224, tangentMode=RCTM_User, time=1, value=1}
       ]}}})
   &SpriteRotationRate(Rotation Rate=&Multiply_Float(
     A=&Lerp_Float(A=-222, B=-55, Alpha=$Particles.NormalizedAge)
     B=&DotProduct(
      Vector A=&ReturnCameraVector(Camera Vectors=::NewEnumerator3, Coordinate Space=::World)
      Vector B=$Particles.Initial.ShapeLocation.ShapeBitangent
     )
    ))
   &ScaleColor(
    Scale Alpha=&FloatFromCurve(FloatCurve=DataInterface() {propertyValues {Curve {keys [{interpMode=RCIM_Cubic, leaveTangent=13.0812, tangentMode=RCTM_Break, time=0, value=0}, <at 0.05 1 RCIM_Cubic>, <at 1 0 RCIM_Cubic>]}}})
    Scale RGB=&VectorFromFloat(Value=&FloatFromCurve(FloatCurve=DataInterface() {propertyValues {Curve {keys [<at 0 1>,<at 0.9983 0.5>]}}}, Scale Curve=$System.TimeOfDayBrightnessMultiplier))
    ScaleA=true
    ScaleRGB=true
   )
  ]
  rendererClasses [&NiagaraSpriteRendererProperties]
  renderers [{rendererClass=&NiagaraSpriteRendererProperties}]
 }
 Flash {
  emitterUpdateScript [
   &EmitterState(Life Cycle Mode=::NewEnumerator1, Loop Behavior=::NewEnumerator1, Loop Duration Mode=::NewEnumerator1, UseLoopDelay=false)
   &SampleSkeletalMesh(
    AdditionalYawPitchRoll=false
    Bone Sampling Mode:ENiagaraBoneSamplingMode=::NewEnumerator4
    Position Offset Coordinate Space:ENiagaraCoordinateSpace=::Local
    Sampled Position Offset:Vector3f=<-10 7 0>
    Skeletal Mesh:NiagaraDataInterfaceSkeletalMesh=$System.Skeletal Mesh
   )
   /SetVariables(Emitter.bool:NiagaraBool=&SetBoolByFloatComparison(
     A=&MakeFloatFromVector(Vector=$Emitter.SampleSkeletalMesh.SampledPositionWithoutOffset, Channel=::NewEnumerator0)
     B=-47
     Comparison Type=::NewEnumerator6
    )) {isSetParameters=true, module=SetVariables_#3}
   /SetVariables(Emitter.Spawn:NiagaraBool=&Boolean_AndOperation(Bool1=&Boolean_NotOperation(Bool=$Emitter.boolLastFrame), Bool2=$Emitter.bool)) {isSetParameters=true, module=SetVariables_#4}
   /SetVariables(Emitter.boolLastFrame:NiagaraBool=$Emitter.bool) {isSetParameters=true, module=SetVariables_#5}
   &SpawnPerFrame(Spawn:NiagaraBool=$Emitter.Spawn, Spawn Count=3, Use Spawn Probability=false)
  ]
  particleSpawnScript [
   &InitializeParticle(
    Color=<rgba 1 1 1 0.003>
    Lifetime=0.05
    Position Offset=$Emitter.SampleSkeletalMesh.SampledPosition
    Sprite Size Mode=::NewEnumerator3
    Uniform Sprite Size=40
    UsePositionOffset=true
    Write Material Random=true
   ) {module=InitializeParticle001}
   &DynamicMaterialParameters(
    Index 0 Param 1=1
    Index 0 Param 2=1
    Index 0 Param 3=1
    Index 0 Param 4=1
    Module.Param0WriteEnabled=true
    Module.Param1WriteEnabled=true
    Module.Param2WriteEnabled=true
    Module.Param3WriteEnabled=true
   )
  ]
  particleUpdateScript [
   &CameraOffset(Camera Offset Amount=10)
   &ParticleState()
   &SolveForcesAndVelocity(Clamp Velocity=false, Limit Acceleration=false)
   &ScaleColor(
    Scale Alpha=&Multiply_Float(
     A=&FloatFromCurve(FloatCurve=DataInterface() {propertyValues {Curve {keys [<at 0 1>,<at 1 0>]}}})
     B=&FNTimeOfDayScalar(Night Scale=0.3, Morning Scale=0.2, Day Scale=0.5, Evening Scale=0.3)
    )
    Scale RGB=&MakeVectorFromLinearColorRGB(Color=<rgba 1 0.9671 0.9772 1>)
    ScaleA=true
    ScaleRGB=true
   )
   &ScaleSpriteSize(Scale Factor=&Vector2DFromFloat(Value=&FloatFromCurve(FloatCurve=DataInterface() {propertyValues {Curve {keys [<at 0 1>,<at 1 3>]}}})))
  ]
  rendererClasses [&NiagaraSpriteRendererProperties]
  renderers [{rendererClass=&NiagaraSpriteRendererProperties}]
 }
 Impact_Spikes {
  emitterUpdateScript [
   &EmitterState(Life Cycle Mode=::NewEnumerator1, Loop Behavior=::NewEnumerator1, Loop Delay=0.03, Loop Duration Mode=::NewEnumerator1, UseLoopDelay=true)
   &SampleSkeletalMesh(
    AdditionalYawPitchRoll=false
    Bone Sampling Mode:ENiagaraBoneSamplingMode=::NewEnumerator4
    Position Offset Coordinate Space:ENiagaraCoordinateSpace=::Local
    Sampled Position Offset:Vector3f=<-10 7 0>
    Skeletal Mesh:NiagaraDataInterfaceSkeletalMesh=$System.Skeletal Mesh
   )
   /SetVariables(Emitter.bool:NiagaraBool=&SetBoolByFloatComparison(
     A=&MakeFloatFromVector(Vector=$Emitter.SampleSkeletalMesh.SampledPositionWithoutOffset, Channel=::NewEnumerator0)
     B=-47
     Comparison Type=::NewEnumerator6
    )) {isSetParameters=true, module=SetVariables_#3}
   /SetVariables(Emitter.Spawn:NiagaraBool=&Boolean_AndOperation(Bool1=&Boolean_NotOperation(Bool=$Emitter.boolLastFrame), Bool2=$Emitter.bool)) {isSetParameters=true, module=SetVariables_#4}
   /SetVariables(Emitter.boolLastFrame:NiagaraBool=$Emitter.bool) {isSetParameters=true, module=SetVariables_#5}
   &SpawnPerFrame(Spawn:NiagaraBool=$Emitter.Spawn, Spawn Count=7, Use Spawn Probability=false)
  ]
  particleSpawnScript [
   &InitializeParticle(
    Color=<rgba 1 1 1 0.075>
    Lifetime=&Multiply_Float(A=&RandomRangeFloat(Minimum=0.05, Maximum=0.08), B=2)
    Position=$Emitter.SampleSkeletalMesh.SampledPosition
    Position Mode=::NewEnumerator1
    Sprite Size Max=<8 30>
    Sprite Size Min=<6 15>
    Sprite Size Mode=::NewEnumerator2
    UsePositionOffset=false
    Write Material Random=true
   )
   &SubUVAnimation(End Frame=3, SubUV Animation Mode:ENiagara_SubUVLookupMode=::NewEnumerator2)
   &ShapeLocation(
    MaskSpawnGroup=false
    Override Seed=false
    Ring Radius=1
    Rotation Angle=1
    Rotation Mode=::NewEnumerator1
    Shape Primitive=::NewEnumerator4
    U Distribution=0.5
    UseOwnerScale=false
   ) {module=ShapeLocation001}
   &AddVelocity(Velocity=$Particles.ShapeLocation001.ShapeTangent, Velocity Speed Scale=25) {module=AddVelocity001}
  ]
  particleUpdateScript [
   &ParticleState()
   &Drag(Drag=3, Use Linear Drag=true, Use Rotational Drag=false)
   &SolveForcesAndVelocity(Clamp Velocity=false, Limit Acceleration=false)
   &ScaleSpriteSize(
    Non-Uniform Curve Scale=&Multiply_Vector2D_ByFloat(Float=1, Vector2D=<1 1>)
    Non-Uniform Curve Sprite Scale=DataInterface() {propertyValues {
      XCurve {keys [{tangentMode=RCTM_User, time=0, value=1}, {tangentMode=RCTM_User, time=1, value=0.8}]}
      YCurve {keys [{tangentMode=RCTM_User, time=0, value=0.5}, {tangentMode=RCTM_User, time=1, value=1.7}]}
     }}
    Scale Sprite Size Mode=::NewEnumerator3
   )
   &ScaleColor(
    Scale Alpha=&Multiply_Float(
     A=&FloatFromCurve(
      FloatCurve=DataInterface() {propertyValues {Curve {keys [{arriveTangent=-2.7319, interpMode=RCIM_Cubic, time=0, value=1}, {arriveTangent=-1, leaveTangent=-1, time=0.2, value=0.05}, <at 1 0>]}}}
      Scale Curve=&FNTimeOfDayScalar(Night Scale=1, Morning Scale=0.3, Evening Scale=0.5)
     )
     B=1
    )
    ScaleA=true
    ScaleRGB=true
   )
   /SetVariables(Particles.Color2=&MultiplyLinearColor(Linear Color=$Particles.Color, Scale Factor=<1 1 1 0.2>)) {isSetParameters=true}
   &SpriteFacingAndAlignment(Facing Coordinate Space=::World, Sprite Facing=<0 0 1>, WriteAlignment=false, WriteFacing=true)
  ]
  rendererClasses [&NiagaraSpriteRendererProperties]
  renderers [
   {rendererClass=&NiagaraSpriteRendererProperties}
   {rendererClass=&NiagaraSpriteRendererProperties, rendererIndex=1}
  ]
 }
}
summary {
 systemName=NS_Emote_OneInchPunch_DustPunch
 emitters [
  {emitterName=DustPunch, rendererClasses [&NiagaraSpriteRendererProperties]}
  {emitterName=Impact_Spikes, rendererClasses [&NiagaraSpriteRendererProperties]}
  {emitterName=Flash, rendererClasses [&NiagaraSpriteRendererProperties]}
  {emitterName=Airflow, rendererClasses [&NiagaraMeshRendererProperties]}
 ]
 userVariables [
  User.bIsFrontEnd {type=NiagaraBool}
  User.bIsFrontEndPreview {type=NiagaraBool}
 ]
}
